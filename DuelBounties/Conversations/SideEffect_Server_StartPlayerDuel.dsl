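; Conversation side effect that, on the server, enters the talking player into the duel bounty queue.
; It takes the player controller from the conversation context and applies an optional instigator gameplay effect to that player's ability system.
; It then finds the DuelBounties component on the Athena game state and calls TryAddPlayertoDuelQueue with the duel quest definition and Phoebe gameplay effect class, printing a debug message if the component is missing.
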(fn ApplyDuelGameplayEffects
  (params (PlayerController :type Fort%20Player%20Controller%20Object%20Reference))
  (bind _returnvalue (Ability|GetFortAbilitySystemComponent PlayerController))
  (bind _optionalduelinstigatorge (Variables|Default|GetOptionalDuelInstigatorGE))
  ;[0 0 0] Apply GEs needed for the duel to the instigator if they are specified
  (Utilities|IsValid _returnvalue
    (:Is%20Valid
      (LocalVariables|SetPlayerAbilitySystem _returnvalue)
      (if (Utilities|IsValidClass _optionalduelinstigatorge)
        (GameplayEffects|ApplyGameplayEffectToSelf :self (LocalVariables|GetPlayerAbilitySystem) :GameplayEffectClass _optionalduelinstigatorge :Level 0.0)))))

(fn ServerCauseSideEffect
  (params (Context :type Conversation%20Context%20Structure%20%28by%20ref%29))
  ;[0 0 0] Apply any duel-specific GEs to instigator and add them to duel queue
  (bind _returnvalue (Class|FortConversationContextLibrary|GetParticipantPlayerController Context))
  (Utilities|IsValid _returnvalue
    (:Is%20Valid
      (bind _output_get (LocalVariables|SetInstigatingPlayerController _returnvalue))
      (CallFunction|ApplyDuelGameplayEffects _output_get)
      (bind _asduel_bounties_game_state_component (Utilities|Casting|CastToDuelBounties_GameStateComponent (Actor|GetComponentbyClass (Fortnite|Athena|GameState|GetGameStateAthena) (LocalVariables|GetGameStateComponent)))
        (:then
          (Class|DuelBountiesGameStateComponent|TryAddPlayertoDuelQueue _asduel_bounties_game_state_component (LocalVariables|GetInstigatingPlayerController) (Variables|Default|GetDuel%20Quest%20Definition) (Variables|Default|GetPhoebe%20Gameplay%20Effect%20Class))
          (return))
        (:CastFailed
          (Development|PrintString "DuelBounties: Game State Component not found" true true "(R=1.000000,G=0.000000,B=0.607688,A=1.000000)" 4.0))))))
